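# Activatable modal panel showing the Donut challenge letter from "Wade": letter art over a rarity flare, with three comic-bubble text blocks on a blurred background.
# PreConstruct and Construct tint the flare's dynamic material from two color variables; Construct also reverse-plays Intro from its end time.
# On activation, a delayed Intro animates in blur, art, flare and bubbles, then reveals the OK button.
# OK and the touch-only close button both pop the modal.
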WidgetBlueprint /BRCosmetics/Athena/UI/Challenges/DonutLetter
animations [Intro {
  binding_count=7
  length_seconds=0.8
  object=&self:Intro
  sequence {bindings [
    ImageFlare {
     id=#1
     possessed_class=Image
     tracks [{
       type=MovieSceneFloatTrack
       display_name=Render Opacity
       sections [{
         type=MovieSceneFloatSection
         end_frame=16
         start_frame=0
         channels [None {type=MovieSceneScriptingFloatChannel, keys [[0,0],[6.4007,0],[16,1]]}]
        }]
      }]
    }
    Border_52 {id=#2, possessed_class=Border, tracks [{type=MovieSceneWidgetMaterialTrack, sections []}]}
    Border_57 {
     id=#3
     possessed_class=Border
     tracks [
      {
       type=MovieSceneFloatTrack
       display_name=Render Opacity
       sections [{
         type=MovieSceneFloatSection
         end_frame=16
         start_frame=0
         channels [None {type=MovieSceneScriptingFloatChannel, thinned_from=5, keys [[0,0],[10.849,0],[13.998,1],[16,1]]}]
        }]
      }
      {
       type=MovieScene2DTransformTrack
       sections [{
         type=MovieScene2DTransformSection
         end_frame=16
         start_frame=4
         channels [
          Translation.X {type=MovieSceneScriptingFloatChannel, const [107.5543,2]}
          Translation.Y {type=MovieSceneScriptingFloatChannel, keys [[10.997,132.9199],[14,39.0303],[14.6313,50.6707],[15.7617,43.9199]]}
          Angle {type=MovieSceneScriptingFloatChannel}
          Scale.X {type=MovieSceneScriptingFloatChannel}
          Scale.Y {type=MovieSceneScriptingFloatChannel}
          Shear.X {type=MovieSceneScriptingFloatChannel}
          Shear.Y {type=MovieSceneScriptingFloatChannel}
         ]
        }]
      }
     ]
    }
    Border_56 {
     id=#4
     possessed_class=Border
     tracks [
      {
       type=MovieScene2DTransformTrack
       sections [{
         type=MovieScene2DTransformSection
         end_frame=13
         start_frame=3
         channels [
          Translation.X {type=MovieSceneScriptingFloatChannel, keys [[7.781,89],[11.8253,-4.2524],[12.5563,4.8598],[13.39,0]]}
          Translation.Y {type=MovieSceneScriptingFloatChannel, const [0,2]}
          Angle {type=MovieSceneScriptingFloatChannel}
          Scale.X {type=MovieSceneScriptingFloatChannel}
          Scale.Y {type=MovieSceneScriptingFloatChannel}
          Shear.X {type=MovieSceneScriptingFloatChannel}
          Shear.Y {type=MovieSceneScriptingFloatChannel}
         ]
        }]
      }
      {
       type=MovieSceneFloatTrack
       display_name=Render Opacity
       sections [{
         type=MovieSceneFloatSection
         end_frame=16
         start_frame=0
         channels [None {type=MovieSceneScriptingFloatChannel, thinned_from=5, keys [[0,0],[7.748,0],[11.8253,1],[16,1]]}]
        }]
      }
     ]
    }
    Image_239 {
     id=#5
     possessed_class=Image
     tracks [
      {
       type=MovieSceneFloatTrack
       display_name=Render Opacity
       sections [{
         type=MovieSceneFloatSection
         end_frame=16
         start_frame=0
         channels [None {type=MovieSceneScriptingFloatChannel, keys [[0,0],[3,0],[8,1],[16,1]]}]
        }]
      }
      {
       type=MovieScene2DTransformTrack
       sections [{
         type=MovieScene2DTransformSection
         end_frame=16
         start_frame=3
         channels [
          Angle {type=MovieSceneScriptingFloatChannel}
          Scale.X {type=MovieSceneScriptingFloatChannel, keys [[3,0.8],[6,1.05],[8,1]]}
          Scale.Y {type=MovieSceneScriptingFloatChannel, keys [[3,0.8],[6,1.05],[8,1]]}
          Shear.X {type=MovieSceneScriptingFloatChannel}
          Shear.Y {type=MovieSceneScriptingFloatChannel}
         ]
        }]
      }
     ]
    }
    Border_55 {
     id=#6
     possessed_class=Border
     tracks [
      {
       type=MovieScene2DTransformTrack
       sections [{
         type=MovieScene2DTransformSection
         end_frame=11
         start_frame=0
         channels [
          Translation.X {type=MovieSceneScriptingFloatChannel, keys [[4.2853,-250],[9.033,-107.6851],[10.1147,-124.3149],[11.449,-116]]}
          Translation.Y {type=MovieSceneScriptingFloatChannel, const [0,2]}
          Angle {type=MovieSceneScriptingFloatChannel}
          Scale.X {type=MovieSceneScriptingFloatChannel}
          Scale.Y {type=MovieSceneScriptingFloatChannel}
          Shear.X {type=MovieSceneScriptingFloatChannel}
          Shear.Y {type=MovieSceneScriptingFloatChannel}
         ]
        }]
      }
      {
       type=MovieSceneFloatTrack
       display_name=Render Opacity
       sections [{
         type=MovieSceneFloatSection
         end_frame=16
         start_frame=0
         channels [None {type=MovieSceneScriptingFloatChannel, keys [[0,0],[4.2853,0],[9.033,1],[16,1]]}]
        }]
      }
     ]
    }
    BackgroundBlur_94 {
     id=#7
     possessed_class=BackgroundBlur
     tracks [{
       type=MovieSceneFloatTrack
       display_name=Blur Strength
       sections [{type=MovieSceneFloatSection, end_frame=5, start_frame=0, channels [None {type=MovieSceneScriptingFloatChannel, keys [[0,0],[5,10]]}]}]
      }]
    }
   ]}
 }]
graphs [EventGraph {nodes [
   n1 = "On Button Clicked (Mobile_ButtonClose)"!() -> (OutputDelegate, then, Button:Common Button Legacy Object Reference) {is=title K2Node_ComponentBoundEvent}
   PopCurrentModal!(execute=[@n1.then, @n2.then]) -> (then)
   n2 = OnBackClicked!() -> (OutputDelegate, then, Button:Common Button Legacy Object Reference) {is=title K2Node_CustomEvent}
   Event Construct()
   > Bind Event to On Button Clicked(self:Common Button Legacy Object Reference=Get Button_Accept() -> (Button_Accept:Icon Text Button Object Reference), Delegate=@n2) {is=title K2Node_AddDelegate}
   n3 = Get Intro()
   > PlayAnimation(InAnimation=@n3, StartAtTime=GetEndTime(self=@n3), PlayMode=Reverse)
   > n4 = GetDynamicMaterial(self=Get ImageFlare() -> (ImageFlare:Image Object Reference))
   > SetVectorParameterValue(self=@n4, ParameterName=Rarity Color 1, Value=Get FlareRGB_1() -> (FlareRGB_1:Linear Color Structure))
   > SetVectorParameterValue(self=@n4, ParameterName=Rarity Color 2, Value=Get FlareRGB_2() -> (FlareRGB_2:Linear Color Structure))
   Event On Activated()
   > Delay(Duration=0.3)
   > PlayAnimationForward(InAnimation=Get Intro())
   > Delay(Duration=0.5)
   > SetVisibility(self=Get Button_Accept(), InVisibility=SelfHitTestInvisible)
   Event PreConstruct()
   > n5 = GetDynamicMaterial(self=Get ImageFlare())
   > SetVectorParameterValue(self=@n5, ParameterName=Rarity Color 1, Value=Get FlareRGB_1())
   > SetVectorParameterValue(self=@n5, ParameterName=Rarity Color 2, Value=Get FlareRGB_2())
  ]}]
variables [
 FlareRGB_1 {type=LinearColor, default=<rgba 0.9867 1 0.7671 1>}
 FlareRGB_2 {type=LinearColor, default=<rgba 1 0.9683 0.6185 1>}
]
widgets {
 info {inheritedWidgetCount=0, parentClass=/Script/FortniteUI.FortActivatablePanel, rootWidgetClass=&Overlay, widgetCount=26}
 widgets [
  Overlay_22 {widgetClassPath=&Overlay, widget=Overlay_22}
  BackgroundBlur_94 {
   widgetClassPath=&BackgroundBlur
   parent=Overlay_22
   widget=BackgroundBlur_94
   properties {
    blurStrength=10
    visibility=Visible
    lowQualityFallbackBrush {
     bIsDynamicallyLoaded=false
     drawAs=Image
     imageType=NoImage
     mirroring=NoMirror
     resourceName=None
     resourceObject=null
     tiling=NoTile
     imageSize=<0 0>
     margin=<ltrb 0 0 0 0>
     outlineSettings {
      bUseBrushTransparency=false
      roundingType=HalfHeightRadius
      width=0
      color {colorUseRule=UseColor_Specified, specifiedColor=<rgba 0 0 0 0>}
      cornerRadii=<0 0 0 0>
     }
     tintColor {colorUseRule=UseColor_Specified, specifiedColor=<rgba 0 0 0 0.4>}
     uVRegion {bIsValid=false, max=<0 0>, min=<0 0>}
    }
   }
   slot {class=OverlaySlot, properties {horizontalAlignment=HAlign_Fill, verticalAlignment=VAlign_Fill}}
  }
  SafeZone_0 {
   widgetClassPath=&SafeZone
   bIsVariable=true
   parent=Overlay_22
   widget=SafeZone_0
   slot {class=OverlaySlot, properties {horizontalAlignment=HAlign_Fill, verticalAlignment=VAlign_Fill}}
  }
  Overlay_24 {widgetClassPath=&Overlay, parent=SafeZone_0, widget=Overlay_24, slot {class=SafeZoneSlot}}
  Overlay_0 {
   widgetClassPath=&Overlay
   parent=Overlay_24
   widget=Overlay_0
   slot {class=OverlaySlot, properties {verticalAlignment=VAlign_Center, padding=<ltrb 100 0 0 0>}}
  }
  ImageFlare {
   widgetClassPath=&Image
   bIsVariable=true
   parent=Overlay_0
   widget=ImageFlare
   properties {brush {
     bIsDynamicallyLoaded=false
     drawAs=Image
     imageType=NoImage
     mirroring=NoMirror
     resourceName=None
     resourceObject=/Game/UI/Frontend/Rewards/materials/M_UI_DailyReward_Rarity-Flares
     tiling=NoTile
     imageSize=<922.0399 1008.3647>
     margin=<ltrb 0 0 0 0>
     outlineSettings {
      bUseBrushTransparency=false
      roundingType=HalfHeightRadius
      width=0
      color {colorUseRule=UseColor_Specified, specifiedColor=<rgba 0 0 0 0>}
      cornerRadii=<0 0 0 0>
     }
     tintColor {colorUseRule=UseColor_Specified, specifiedColor=<rgba 1 1 1 1>}
     uVRegion {bIsValid=false, max=<0 0>, min=<0 0>}
    }}
   slot {class=OverlaySlot, properties {horizontalAlignment=HAlign_Center, verticalAlignment=VAlign_Center}}
  }
  Image_239 {
   widgetClassPath=&Image
   bIsVariable=true
   parent=Overlay_0
   widget=Image_239
   properties {
    brush {
     bIsDynamicallyLoaded=false
     drawAs=Image
     imageType=NoImage
     mirroring=NoMirror
     resourceName=None
     resourceObject=/BRCosmetics/Athena/UI/Challenges/Art/Donut/T_UI_DonutLetter
     tiling=NoTile
     imageSize=<1024 1024>
     margin=<ltrb 0 0 0 0>
     outlineSettings {
      bUseBrushTransparency=false
      roundingType=HalfHeightRadius
      width=0
      color {colorUseRule=UseColor_Specified, specifiedColor=<rgba 0 0 0 0>}
      cornerRadii=<0 0 0 0>
     }
     tintColor {colorUseRule=UseColor_Specified, specifiedColor=<rgba 1 1 1 1>}
     uVRegion {bIsValid=false, max=<0 0>, min=<0 0>}
    }
    renderTransform {angle=-3, scale=<1 1>, shear=<0 0>, translation=<0 0>}
   }
   slot {class=OverlaySlot}
  }
  VerticalBox_3 {
   widgetClassPath=&VerticalBox
   parent=Overlay_24
   widget=VerticalBox_3
   slot {class=OverlaySlot, properties {horizontalAlignment=HAlign_Right, verticalAlignment=VAlign_Bottom, padding=<ltrb 0 0 150 0>}}
  }
  Border_48 {
   widgetClassPath=&Border
   bIsVariable=true
   parent=VerticalBox_3
   widget=Border_48
   properties {
    background {
     bIsDynamicallyLoaded=false
     drawAs=Image
     imageType=NoImage
     mirroring=NoMirror
     resourceName=None
     resourceObject=/Game/Environments/FrontEnd/BackPlates/UITakeovers/M_UI_ComicBubble_Inst
     tiling=NoTile
     imageSize=<32 32>
     margin=<ltrb 0 0 0 0>
     outlineSettings {
      bUseBrushTransparency=false
      roundingType=HalfHeightRadius
      width=0
      color {colorUseRule=UseColor_Specified, specifiedColor=<rgba 0 0 0 0>}
      cornerRadii=<0 0 0 0>
     }
     tintColor {colorUseRule=UseColor_Specified, specifiedColor=<rgba 1 1 1 1>}
     uVRegion {bIsValid=false, max=<0 0>, min=<0 0>}
    }
    brushColor=<rgba 1 1 1 0>
    padding=<ltrb 10 10 10 10>
   }
   slot {class=VerticalBoxSlot}
  }
  VerticalBox_0 {
   widgetClassPath=&VerticalBox
   bIsVariable=true
   parent=Border_48
   widget=VerticalBox_0
   slot {class=BorderSlot, properties {padding=<ltrb 10 10 10 10>}}
  }
  Border_55 {
   widgetClassPath=&Border
   parent=VerticalBox_0
   widget=Border_55
   properties {
    horizontalAlignment=HAlign_Center
    verticalAlignment=VAlign_Center
    brushColor=<rgba 0 0 0 1>
    padding=<ltrb 4 4 4 4>
    renderTransform {angle=0, scale=<1 1>, shear=<0 0>, translation=<-116.0512 0>}
   }
   slot {class=VerticalBoxSlot, properties {horizontalAlignment=HAlign_Left, verticalAlignment=VAlign_Center}}
  }
  Border_52 {
   widgetClassPath=&Border
   parent=Border_55
   widget=Border_52
   properties {
    horizontalAlignment=HAlign_Center
    verticalAlignment=VAlign_Center
    background {
     bIsDynamicallyLoaded=false
     drawAs=Image
     imageType=NoImage
     mirroring=NoMirror
     resourceName=None
     resourceObject=&M_UI_ComicBubble_Inst
     tiling=NoTile
     imageSize=<32 32>
     margin=<ltrb 0 0 0 0>
     outlineSettings {
      bUseBrushTransparency=false
      roundingType=HalfHeightRadius
      width=0
      color {colorUseRule=UseColor_Specified, specifiedColor=<rgba 0 0 0 0>}
      cornerRadii=<0 0 0 0>
     }
     tintColor {colorUseRule=UseColor_Specified, specifiedColor=<rgba 1 1 1 1>}
     uVRegion {bIsValid=false, max=<0 0>, min=<0 0>}
    }
    padding=<ltrb 15 15 15 15>
   }
   slot {class=BorderSlot, properties {horizontalAlignment=HAlign_Center, verticalAlignment=VAlign_Center, padding=<ltrb 4 4 4 4>}}
  }
  VerticalBox_2 {
   widgetClassPath=&VerticalBox
   parent=Border_52
   widget=VerticalBox_2
   slot {class=BorderSlot, properties {horizontalAlignment=HAlign_Center, verticalAlignment=VAlign_Center, padding=<ltrb 15 15 15 15>}}
  }
  CommonRichTextBlock_0 {
   widgetClassPath=&CommonRichTextBlock
   bIsVariable=true
   parent=VerticalBox_2
   widget=CommonRichTextBlock_0
   properties {
    text="Dear <Bold>Fortnite,</>"
    textStyleSet=/Game/Environments/FrontEnd/BackPlates/UITakeovers/DonutChatBubble_RichTextStyles
    defaultTextStyleOverride {
     overflowPolicy=Clip
     transformPolicy=None
     colorAndOpacity {colorUseRule=UseColor_Specified, specifiedColor=<rgba 1 1 1 1>}
     font {
      bForceMonospaced=false
      bMaterialIsStencil=false
      fontMaterial=null
      fontObject=&NotoSans
      letterSpacing=0
      monospacedWidth=1
      size=24
      skewAmount=0
      typefaceFontName=NotoSans-Regular
      outlineSettings {
       bApplyOutlineToDropShadows=false
       bMiteredCorners=false
       bSeparateFillAlpha=false
       outlineMaterial=null
       outlineSize=0
       outlineColor=<rgba 0 0 0 1>
      }
     }
     highlightColor {colorUseRule=UseColor_Specified, specifiedColor=<rgba 0 0 0 1>}
     highlightShape {
      bIsDynamicallyLoaded=false
      drawAs=Image
      imageType=NoImage
      mirroring=NoMirror
      resourceName=None
      resourceObject=null
      tiling=NoTile
      imageSize=<32 32>
      margin=<ltrb 0 0 0 0>
      outlineSettings {
       bUseBrushTransparency=false
       roundingType=HalfHeightRadius
       width=0
       color {colorUseRule=UseColor_Specified, specifiedColor=<rgba 0 0 0 0>}
       cornerRadii=<0 0 0 0>
      }
      tintColor {colorUseRule=UseColor_Specified, specifiedColor=<rgba 1 1 1 1>}
      uVRegion {bIsValid=false, max=<0 0>, min=<0 0>}
     }
     selectedBackgroundColor {colorUseRule=UseColor_Foreground, specifiedColor=<rgba 1 0 1 1>}
     shadowColorAndOpacity=<rgba 0 0 0 1>
     shadowOffset=<0 0>
     strikeBrush {
      bIsDynamicallyLoaded=false
      drawAs=Image
      imageType=NoImage
      mirroring=NoMirror
      resourceName=None
      resourceObject=null
      tiling=NoTile
      imageSize=<32 32>
      margin=<ltrb 0 0 0 0>
      outlineSettings {
       bUseBrushTransparency=false
       roundingType=HalfHeightRadius
       width=0
       color {colorUseRule=UseColor_Specified, specifiedColor=<rgba 0 0 0 0>}
       cornerRadii=<0 0 0 0>
      }
      tintColor {colorUseRule=UseColor_Specified, specifiedColor=<rgba 1 1 1 1>}
      uVRegion {bIsValid=false, max=<0 0>, min=<0 0>}
     }
     underlineBrush {
      bIsDynamicallyLoaded=false
      drawAs=Image
      imageType=NoImage
      mirroring=NoMirror
      resourceName=None
      resourceObject=null
      tiling=NoTile
      imageSize=<32 32>
      margin=<ltrb 0 0 0 0>
      outlineSettings {
       bUseBrushTransparency=false
       roundingType=HalfHeightRadius
       width=0
       color {colorUseRule=UseColor_Specified, specifiedColor=<rgba 0 0 0 0>}
       cornerRadii=<0 0 0 0>
      }
      tintColor {colorUseRule=UseColor_Specified, specifiedColor=<rgba 1 1 1 1>}
      uVRegion {bIsValid=false, max=<0 0>, min=<0 0>}
     }
    }
    renderTransform {angle=0, scale=<1 1>, shear=<-4 0>, translation=<0 0>}
   }
   slot {class=VerticalBoxSlot}
  }
  CommonTextBlock {
   widgetClassPath=&CommonTextBlock
   bIsVariable=true
   parent=VerticalBox_2
   widget=CommonTextBlock
   properties {
    text=I came up with a cool redesign for the battle bus.
    wrapTextAt=500
    colorAndOpacity {colorUseRule=UseColor_Specified, specifiedColor=<rgba 0 0 0 1>}
    font {
     bForceMonospaced=false
     bMaterialIsStencil=false
     fontMaterial=null
     fontObject=&BurbankSmall
     letterSpacing=0
     monospacedWidth=1
     size=28
     skewAmount=0
     typefaceFontName=Bold
     outlineSettings {
      bApplyOutlineToDropShadows=false
      bMiteredCorners=false
      bSeparateFillAlpha=false
      outlineMaterial=null
      outlineSize=0
      outlineColor=<rgba 0 0 0 1>
     }
    }
    renderTransform {angle=0, scale=<1 1>, shear=<-4 0>, translation=<0 0>}
    shadowOffset=<0 0>
   }
   slot {class=VerticalBoxSlot, properties {padding=<ltrb 0 15 0 0>}}
  }
  Border_56 {
   widgetClassPath=&Border
   parent=VerticalBox_0
   widget=Border_56
   properties {brushColor=<rgba 0 0 0 1>, padding=<ltrb 4 4 4 4>}
   slot {class=VerticalBoxSlot, properties {padding=<ltrb 0 20 0 0>}}
  }
  Border_53 {
   widgetClassPath=&Border
   parent=Border_56
   widget=Border_53
   properties {
    background {
     bIsDynamicallyLoaded=false
     drawAs=Image
     imageType=NoImage
     mirroring=NoMirror
     resourceName=None
     resourceObject=&M_UI_ComicBubble_Inst
     tiling=NoTile
     imageSize=<32 32>
     margin=<ltrb 0 0 0 0>
     outlineSettings {
      bUseBrushTransparency=false
      roundingType=HalfHeightRadius
      width=0
      color {colorUseRule=UseColor_Specified, specifiedColor=<rgba 0 0 0 0>}
      cornerRadii=<0 0 0 0>
     }
     tintColor {colorUseRule=UseColor_Specified, specifiedColor=<rgba 1 1 1 1>}
     uVRegion {bIsValid=false, max=<0 0>, min=<0 0>}
    }
    padding=<ltrb 15 15 15 15>
   }
   slot {class=BorderSlot, properties {padding=<ltrb 4 4 4 4>}}
  }
  CommonRichTextBlock_1 {
   widgetClassPath=&CommonRichTextBlock
   bIsVariable=true
   parent=Border_53
   widget=CommonRichTextBlock_1
   properties {
    text="I bet my friends my good <Bold>drawing arm</> you'd put this in the game last season.\r\nPlease tell your artist <Bold>limbs are on the line</> and to add this <Bold>right away.</>"
    textStyleSet=&DonutChatBubble_RichTextStyles
    wrapTextAt=600
    defaultTextStyleOverride {
     overflowPolicy=Clip
     transformPolicy=None
     colorAndOpacity {colorUseRule=UseColor_Specified, specifiedColor=<rgba 1 1 1 1>}
     font {
      bForceMonospaced=false
      bMaterialIsStencil=false
      fontMaterial=null
      fontObject=&NotoSans
      letterSpacing=0
      monospacedWidth=1
      size=24
      skewAmount=0
      typefaceFontName=NotoSans-Regular
      outlineSettings {
       bApplyOutlineToDropShadows=false
       bMiteredCorners=false
       bSeparateFillAlpha=false
       outlineMaterial=null
       outlineSize=0
       outlineColor=<rgba 0 0 0 1>
      }
     }
     highlightColor {colorUseRule=UseColor_Specified, specifiedColor=<rgba 0 0 0 1>}
     highlightShape {
      bIsDynamicallyLoaded=false
      drawAs=Image
      imageType=NoImage
      mirroring=NoMirror
      resourceName=None
      resourceObject=null
      tiling=NoTile
      imageSize=<32 32>
      margin=<ltrb 0 0 0 0>
      outlineSettings {
       bUseBrushTransparency=false
       roundingType=HalfHeightRadius
       width=0
       color {colorUseRule=UseColor_Specified, specifiedColor=<rgba 0 0 0 0>}
       cornerRadii=<0 0 0 0>
      }
      tintColor {colorUseRule=UseColor_Specified, specifiedColor=<rgba 1 1 1 1>}
      uVRegion {bIsValid=false, max=<0 0>, min=<0 0>}
     }
     selectedBackgroundColor {colorUseRule=UseColor_Foreground, specifiedColor=<rgba 1 0 1 1>}
     shadowColorAndOpacity=<rgba 0 0 0 1>
     shadowOffset=<0 0>
     strikeBrush {
      bIsDynamicallyLoaded=false
      drawAs=Image
      imageType=NoImage
      mirroring=NoMirror
      resourceName=None
      resourceObject=null
      tiling=NoTile
      imageSize=<32 32>
      margin=<ltrb 0 0 0 0>
      outlineSettings {
       bUseBrushTransparency=false
       roundingType=HalfHeightRadius
       width=0
       color {colorUseRule=UseColor_Specified, specifiedColor=<rgba 0 0 0 0>}
       cornerRadii=<0 0 0 0>
      }
      tintColor {colorUseRule=UseColor_Specified, specifiedColor=<rgba 1 1 1 1>}
      uVRegion {bIsValid=false, max=<0 0>, min=<0 0>}
     }
     underlineBrush {
      bIsDynamicallyLoaded=false
      drawAs=Image
      imageType=NoImage
      mirroring=NoMirror
      resourceName=None
      resourceObject=null
      tiling=NoTile
      imageSize=<32 32>
      margin=<ltrb 0 0 0 0>
      outlineSettings {
       bUseBrushTransparency=false
       roundingType=HalfHeightRadius
       width=0
       color {colorUseRule=UseColor_Specified, specifiedColor=<rgba 0 0 0 0>}
       cornerRadii=<0 0 0 0>
      }
      tintColor {colorUseRule=UseColor_Specified, specifiedColor=<rgba 1 1 1 1>}
      uVRegion {bIsValid=false, max=<0 0>, min=<0 0>}
     }
    }
    renderTransform {angle=0, scale=<1 1>, shear=<-4 0>, translation=<0 0>}
   }
   slot {class=BorderSlot, properties {padding=<ltrb 15 15 15 15>}}
  }
  Border_57 {
   widgetClassPath=&Border
   parent=VerticalBox_0
   widget=Border_57
   properties {brushColor=<rgba 0 0 0 1>, padding=<ltrb 4 4 4 4>, renderTransform {angle=0, scale=<1 1>, shear=<0 0>, translation=<55 43.9199>}}
   slot {class=VerticalBoxSlot, properties {horizontalAlignment=HAlign_Center, padding=<ltrb 0 30 0 0>}}
  }
  Border_54 {
   widgetClassPath=&Border
   parent=Border_57
   widget=Border_54
   properties {background {
     bIsDynamicallyLoaded=false
     drawAs=Image
     imageType=NoImage
     mirroring=NoMirror
     resourceName=None
     resourceObject=&M_UI_ComicBubble_Inst
     tiling=NoTile
     imageSize=<32 32>
     margin=<ltrb 0 0 0 0>
     outlineSettings {
      bUseBrushTransparency=false
      roundingType=HalfHeightRadius
      width=0
      color {colorUseRule=UseColor_Specified, specifiedColor=<rgba 0 0 0 0>}
      cornerRadii=<0 0 0 0>
     }
     tintColor {colorUseRule=UseColor_Specified, specifiedColor=<rgba 1 1 1 1>}
     uVRegion {bIsValid=false, max=<0 0>, min=<0 0>}
    }}
   slot {class=BorderSlot, properties {padding=<ltrb 4 4 4 4>}}
  }
  VerticalBox_1 {widgetClassPath=&VerticalBox, parent=Border_54, widget=VerticalBox_1, slot {class=BorderSlot}}
  CommonTextBlock_1 {
   widgetClassPath=&CommonTextBlock
   bIsVariable=true
   parent=VerticalBox_1
   widget=CommonTextBlock_1
   properties {
    text="-Your best friend,"
    colorAndOpacity {colorUseRule=UseColor_Specified, specifiedColor=<rgba 0 0 0 1>}
    font {
     bForceMonospaced=false
     bMaterialIsStencil=false
     fontMaterial=null
     fontObject=&BurbankSmall
     letterSpacing=0
     monospacedWidth=1
     size=28
     skewAmount=0
     typefaceFontName=Bold
     outlineSettings {
      bApplyOutlineToDropShadows=false
      bMiteredCorners=false
      bSeparateFillAlpha=false
      outlineMaterial=null
      outlineSize=0
      outlineColor=<rgba 0 0 0 1>
     }
    }
    renderTransform {angle=0, scale=<1 1>, shear=<-4 0>, translation=<0 0>}
    shadowOffset=<0 0>
   }
   slot {class=VerticalBoxSlot}
  }
  CommonTextBlock_2 {
   widgetClassPath=&CommonTextBlock
   bIsVariable=true
   parent=VerticalBox_1
   widget=CommonTextBlock_2
   properties {
    text=Wade
    colorAndOpacity {colorUseRule=UseColor_Specified, specifiedColor=<rgba 0 0 0 1>}
    font {
     bForceMonospaced=false
     bMaterialIsStencil=false
     fontMaterial=null
     fontObject=&BurbankSmall
     letterSpacing=0
     monospacedWidth=1
     size=30
     skewAmount=0
     typefaceFontName=Black
     outlineSettings {
      bApplyOutlineToDropShadows=false
      bMiteredCorners=false
      bSeparateFillAlpha=false
      outlineMaterial=null
      outlineSize=0
      outlineColor=<rgba 0 0 0 1>
     }
    }
    renderTransform {angle=0, scale=<1 1>, shear=<-4 0>, translation=<0 0>}
    shadowOffset=<0 0>
   }
   slot {class=VerticalBoxSlot, properties {horizontalAlignment=HAlign_Center, padding=<ltrb 0 6 0 0>}}
  }
  Button_Accept {
   widgetClassPath=&IconTextButton
   bIsVariable=true
   parent=VerticalBox_3
   widget=Button_Accept
   properties {
    buttonText=OK
    onButtonBaseClicked=(&self:WidgetTree.Button_Accept.HandleOnButtonClicked)
    onButtonBaseDoubleClicked=(&self:WidgetTree.Button_Accept.HandleOnButtonDoubleClicked)
    onButtonBaseHovered=(&self:WidgetTree.Button_Accept.HandleOnButtonHovered)
    onButtonBaseUnhovered=(&self:WidgetTree.Button_Accept.HandleOnButtonUnhovered)
    onSelectedChangedBase=(&self:WidgetTree.Button_Accept.HandleOnSelectedChanged)
    style=/BRCosmetics/UI/Frontend/ItemShop/Com/ButtonStyle-Skew_Com_Like_Yellow.ButtonStyle-Skew_Com_Like_Yellow_C
    visibility=Hidden
    triggeringInputAction {dataTable=&InputActionDataTable, rowName=GenericAccept}
   }
   slot {
    class=VerticalBoxSlot
    properties {horizontalAlignment=HAlign_Center, verticalAlignment=VAlign_Bottom, padding=<ltrb 20 100 20 20>, size {sizeRule=Fill, value=1}}
   }
  }
  VisibilityBorder_Touch {
   widgetClassPath=&CommonHardwareVisibilityBorder
   parent=Overlay_22
   widget=VisibilityBorder_Touch
   properties {visibilityQuery {
     autoDescription=" ANY( Input.Touch )"
     tokenStreamVersion=0
     userDescription=""
     queryTokenStream [0,1,1,1,0]
     tagDictionary [{tagName=Input.Touch}]
    }}
   slot {class=OverlaySlot, properties {horizontalAlignment=HAlign_Right, padding=<ltrb 0 32 32 0>}}
  }
  Mobile_ButtonClose {
   widgetClassPath=/FortUILibrary/Foundation/Buttons/CloseButton
   bIsVariable=true
   parent=VisibilityBorder_Touch
   widget=Mobile_ButtonClose
   properties {
    onButtonBaseClicked=(&self:WidgetTree.Mobile_ButtonClose.HandleOnButtonClicked)
    onButtonBaseDoubleClicked=(&self:WidgetTree.Mobile_ButtonClose.HandleOnButtonDoubleClicked)
    onButtonBaseHovered=(&self:WidgetTree.Mobile_ButtonClose.HandleOnButtonHovered)
    onButtonBaseUnhovered=(&self:WidgetTree.Mobile_ButtonClose.HandleOnButtonUnhovered)
    onSelectedChangedBase=(&self:WidgetTree.Mobile_ButtonClose.HandleOnSelectedChanged)
   }
   slot {class=BorderSlot, properties {padding=<ltrb 0 0 0 0>}}
  }
 ]
}
